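# Niagara ribbon trail effect for the BalloonDish bass cosmetic, built from AnimTrail modules that follow mesh bones or sockets.
# Audio_Chroma_Trail tapers and fades, shifting from the user secondary color toward white.
# Trail_Instrument_Color makes two short-lived, offset trails with hard-coded colors, using spawn-group masks to route width to separate renderers.
# Both scale by time of day; a three-trail variant is disabled.
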NiagaraSystem /SparksCosmetics/Instrument/Bass/BalloonDish/Effects/NS_Bass_BalloonDish_Trail
dependencies {usedDataInterfaces [&NiagaraDataInterfaceCurve], usedDynamicInputs [], usedModules [&SystemState]}
emitters {
 Audio_Chroma_Trail {
  emitterUpdateScript [
   &EmitterState(Life Cycle Mode=::NewEnumerator1, Loop Behavior=::NewEnumerator1, Loop Duration=0.25, UseLoopDelay=true)
   &FN_SpawnAnimTrail_v2()
  ]
  particleSpawnScript [
   /Game/Effects/Niagara/Modules/AnimTrail/FN_InitializeAnimTrail/FN_InitializeAnimTrailBoneIndecies(First Bone or Socket Index=1)
   /Game/Effects/Niagara/Modules/AnimTrail/FN_InitializeAnimTrail/FN_InitializeAnimTrail(AnimTrailWidth=1, Color=<rgba 1 1 1 0.5>, Lifetime=0.25)
   &SystemLocation(MaskSpawnGroup=false)
   &DynamicMaterialParameters(
    Index 0 Param 1=1
    Index 0 Param 4=1
    Module.Index0Param0Write=true
    Module.Index0Param1Write=true
    Module.Index0Param2Write=true
    Module.Index0Param3Write=true
   )
   &AddVelocity(Velocity=<0 1000 0>, Velocity Speed Scale=&OneMinusFloat(Float=$Emitter.NormalizedLoopAge)) {enabled=false}
  ]
  particleUpdateScript [
   &SolveForcesAndVelocity(Clamp Velocity=false, Limit Acceleration=false)
   &ScaleRibbonWidth(Ribbon Width Scale=&FloatFromCurve(FloatCurve=DataInterface() {propertyValues {Curve {keys [
         {arriveTangent=-1.2579, interpMode=RCIM_Cubic, time=0, value=1}
         {arriveTangent=-0.9664, interpMode=RCIM_Cubic, leaveTangent=-0.9664, tangentMode=RCTM_User, time=1, value=0}
        ]}}}))
   &ScaleColor(
    Scale Alpha=&Multiply_Float(
     A=&FloatFromCurve(FloatCurve=DataInterface() {propertyValues {Curve {keys [
          {interpMode=RCIM_Cubic, leaveTangent=-0.143, tangentMode=RCTM_Break, time=0, value=1}
          <at 0.6689 0.0233>
          {arriveTangent=-0.2139, interpMode=RCIM_Cubic, tangentMode=RCTM_Break, time=1, value=0}
         ]}}})
     B=&FNTimeOfDayScalar(Night Scale=1, Morning Scale=0.5, Evening Scale=0.5)
    )
    Scale RGB=<500 500 500>
    ScaleA=true
    ScaleRGB=true
   ) {enabled=false, module=ScaleColor002}
   &FN_UpdateAnimTrail_v2()
   &ScaleColor(
    Scale Alpha=&Multiply_Float(
     A=&FloatFromCurve(FloatCurve=DataInterface() {propertyValues {Curve {keys [
          {interpMode=RCIM_Cubic, leaveTangent=-0.143, tangentMode=RCTM_Break, time=0, value=1}
          <at 0.6689 0.0233>
          {arriveTangent=-0.2139, interpMode=RCIM_Cubic, tangentMode=RCTM_Break, time=1, value=0}
         ]}}})
     B=&FNTimeOfDayScalar(Night Scale=0.65, Morning Scale=0.5, Day Scale=0.7, Evening Scale=0.7)
    )
    Scale RGB=&MakeVectorFromLinearColorRGB(Color=&LerpLinearColors(
      StartColor=$User.TrailSecondaryColor
      EndColor=<rgba 1 1 1 1>
      LerpFactor=&ClampFloat(Float=&Multiply_Float(A=1.5, B=$Particles.NormalizedAge))
     ))
    ScaleA=true
    ScaleRGB=true
   )
  ]
  rendererClasses [&NiagaraRibbonRendererProperties]
  renderers [{rendererClass=&NiagaraRibbonRendererProperties}]
 }
 Trail_Instrument_Color {
  emitterUpdateScript [
   &EmitterState()
   &FN_SpawnAnimTrail_v2(Number of Anim Trails=2)
  ]
  particleSpawnScript [
   &FN_InitializeAnimTrailBoneIndecies(First Bone or Socket Index=1)
   &FN_InitializeAnimTrailBoneIndecies(AnimTrail Selection=1, First Bone or Socket Index=1) {module=FN_InitializeAnimTrailBoneIndecies001}
   &FN_InitializeAnimTrail(
    AnimTrailWidth:NiagaraFloat=&MaskFloatBySpawnGroup(Masked Float=1, Passthrough Float=1)
    Color=&MaskLinearColorBySpawnGroup(Masked Linear Color=<rgba 1 1 1 1>, Passthrough Linear Color=<rgba 0.4 0 0 1.5>)
    Lifetime=0.1
    Position Offset:Vector3f=&RotateVectorByQuaternion(
     Vector To Rotate=&MaskVectorBySpawnGroup(Masked Vector=<0 0 -4>, Passthrough Vector=<0 0 -5.5>)
     Quaternion=$Engine.Owner.Rotation
    )
   )
   &SystemLocation(MaskSpawnGroup=false)
   &AddVelocity(Velocity=<0 1000 0>, Velocity Speed Scale=&OneMinusFloat(Float=$Emitter.NormalizedLoopAge)) {enabled=false}
  ]
  particleUpdateScript [
   &SolveForcesAndVelocity(Clamp Velocity=false, Limit Acceleration=false)
   &ScaleRibbonWidth(Ribbon Width Scale=&FloatFromCurve(FloatCurve=DataInterface() {propertyValues {Curve {keys [
         {arriveTangent=-1.2579, interpMode=RCIM_Cubic, time=0, value=1}
         {arriveTangent=-0.9664, interpMode=RCIM_Cubic, leaveTangent=-0.9664, tangentMode=RCTM_User, time=1, value=0}
        ]}}})) {enabled=false}
   &ScaleColor(
    Scale Alpha=&Multiply_Float(
     A=&FloatFromCurve(FloatCurve=DataInterface() {propertyValues {Curve {keys [
          {interpMode=RCIM_Cubic, leaveTangent=-2.3646, tangentMode=RCTM_Break, time=0, value=1}
          <at 0.6689 0.0233>
          {arriveTangent=-0.2139, interpMode=RCIM_Cubic, tangentMode=RCTM_Break, time=1, value=0}
         ]}}})
     B=&FNTimeOfDayScalar(Night Scale=0.5, Morning Scale=0.65, Day Scale=0.7, Evening Scale=0.7)
    )
    Scale RGB=&Multiply_VectorByFloat(
     Vector=&FN_Select_Vector_from_Time_Of_Day(Morning=<0.9 0.9 0.9>, Day=<1 1 1>, Evening=<1 1 1>, Night=<0.5 0.5 1>)
     Float=&MaskFloatBySpawnGroup(
      Masked Float=1
      Passthrough Float=&FloatFromCurve(FloatCurve=DataInterface() {propertyValues {Curve {keys [
           <at 0 0 RCIM_Cubic>
           {arriveTangent=3.7267, interpMode=RCIM_Cubic, leaveTangent=3.7267, tangentMode=RCTM_User, time=0.17, value=0.5}
           <at 0.5 1>
          ]}}})
     )
    )
    ScaleA=true
    ScaleRGB=true
   ) {module=ScaleColor002}
   &FN_UpdateAnimTrail_v2()
   /SetVariables(
    Particles.RibbonWidthA:NiagaraFloat=&MaskFloatBySpawnGroup(Masked Float=$Particles.RibbonWidth, Passthrough Float=0)
    Particles.RibbonWidthB:NiagaraFloat=&MaskFloatBySpawnGroup(Masked Float=0, Passthrough Float=$Particles.RibbonWidth)
   ) {isSetParameters=true, module=SetVariables_#1}
  ]
  rendererClasses [&NiagaraRibbonRendererProperties]
  renderers [
   {rendererClass=&NiagaraRibbonRendererProperties}
   {rendererClass=&NiagaraRibbonRendererProperties, rendererIndex=1}
  ]
 }
 Trail_Instrument_Color001 {
  bEnabled=false
  emitterUpdateScript [
   &EmitterState()
   &FN_SpawnAnimTrail_v2(Number of Anim Trails=3)
  ]
  particleSpawnScript [
   &FN_InitializeAnimTrailBoneIndecies(First Bone or Socket Index=1)
   &FN_InitializeAnimTrailBoneIndecies(AnimTrail Selection=1, Second Bone or Socket Index=1) {module=FN_InitializeAnimTrailBoneIndecies001}
   &FN_InitializeAnimTrailBoneIndecies(AnimTrail Selection=2, First Bone or Socket Index=1) {module=FN_InitializeAnimTrailBoneIndecies002}
   &FN_InitializeAnimTrail(
    AnimTrailWidth:NiagaraFloat=&MaskFloatBySpawnGroup(
     Masked Float=1
     Passthrough Float=&MaskFloatBySpawnGroup(Masked Float=0.35, Passthrough Float=1, Spawn Group Mask=1)
    )
    Color=&MaskLinearColorBySpawnGroup(
     Masked Linear Color=<rgba 1 1 1 1>
     Passthrough Linear Color=&MaskLinearColorBySpawnGroup(Masked Linear Color=<rgba 0.4 0 0 1.5>, Passthrough Linear Color=<rgba 0.0262 0.0331 0.0382 3>, Spawn Group Mask=1)
    )
    Lifetime=0.1
    Position Offset:Vector3f=&RotateVectorByQuaternion(
     Vector To Rotate=&MaskVectorBySpawnGroup(
      Masked Vector=<0 0 -4>
      Passthrough Vector=&MaskVectorBySpawnGroup(Masked Vector=<0 0 25>, Passthrough Vector=<0 0 -5.5>, Spawn Group Mask=1)
     )
     Quaternion=$Engine.Owner.Rotation
    )
   )
   &SystemLocation(MaskSpawnGroup=false)
   &AddVelocity(Velocity=<0 1000 0>, Velocity Speed Scale=&OneMinusFloat(Float=$Emitter.NormalizedLoopAge)) {enabled=false}
  ]
  particleUpdateScript [
   &SolveForcesAndVelocity(Clamp Velocity=false, Limit Acceleration=false)
   &ScaleRibbonWidth(Ribbon Width Scale=&FloatFromCurve(FloatCurve=DataInterface() {propertyValues {Curve {keys [
         {arriveTangent=-1.2579, interpMode=RCIM_Cubic, time=0, value=1}
         {arriveTangent=-0.9664, interpMode=RCIM_Cubic, leaveTangent=-0.9664, tangentMode=RCTM_User, time=1, value=0}
        ]}}})) {enabled=false}
   &ScaleColor(
    Scale Alpha=&Multiply_Float(
     A=&FloatFromCurve(FloatCurve=DataInterface() {propertyValues {Curve {keys [
          {interpMode=RCIM_Cubic, leaveTangent=-2.3646, tangentMode=RCTM_Break, time=0, value=1}
          <at 0.6689 0.0233>
          {arriveTangent=-0.2139, interpMode=RCIM_Cubic, tangentMode=RCTM_Break, time=1, value=0}
         ]}}})
     B=&FNTimeOfDayScalar(Night Scale=0.5, Morning Scale=0.65, Day Scale=0.7, Evening Scale=0.7)
    )
    Scale RGB=&FN_Select_Vector_from_Time_Of_Day(Morning=<0.9 0.9 0.9>, Day=<1 1 1>, Evening=<1 1 1>, Night=<0.5 0.5 1>)
    ScaleA=true
    ScaleRGB=true
   ) {module=ScaleColor002}
   &FN_UpdateAnimTrail_v2()
   /SetVariables(
    Particles.RibbonWidthA:NiagaraFloat=&MaskFloatBySpawnGroup(Masked Float=0, Passthrough Float=$Particles.RibbonWidth, Spawn Group Mask=2)
    Particles.RibbonWidthB:NiagaraFloat=&MaskFloatBySpawnGroup(Masked Float=$Particles.RibbonWidth, Passthrough Float=0, Spawn Group Mask=2)
   ) {isSetParameters=true, module=SetVariables_#1}
  ]
  rendererClasses [&NiagaraRibbonRendererProperties]
  renderers [
   {rendererClass=&NiagaraRibbonRendererProperties}
   {rendererClass=&NiagaraRibbonRendererProperties, rendererIndex=1}
  ]
 }
}
summary {
 systemName=NS_Bass_BalloonDish_Trail
 emitters [
  {emitterName=Trail_Instrument_Color, rendererClasses [&NiagaraRibbonRendererProperties]}
  {emitterName=Audio_Chroma_Trail, rendererClasses [&NiagaraRibbonRendererProperties]}
  {bEnabled=false, emitterName=Trail_Instrument_Color001, rendererClasses [&NiagaraRibbonRendererProperties]}
 ]
 userVariables [
  User.RibbonIDOffset {type=NiagaraInt32}
  User.SkeletalMesh {
   type=NiagaraDataInterfaceSkeletalMesh
   defaultValue {
    struct=NiagaraExt_VariableValue_DataInterface
    value {dataInterface=&self:NiagaraDataInterfaceSkeletalMesh_0}
   }
  }
  User.TrailColor {type=LinearColor, description="", defaultValue {struct=LinearColor, value=<rgba 1 0 0 2>}}
  User.TrailSecondaryColor {type=LinearColor, description="", defaultValue {struct=LinearColor, value=<rgba 0.899 0.899 1 0.4539>}}
 ]
}
